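# Sound cue for boulder physics impacts, playing bounce sounds with boulder impact attenuation in the Fort_Impacts class.
# An ImpactType integer switch selects small, medium or large sets of randomized waves, with the small set also filling the default slot.
# Each set's continuous modulator uses a Strength parameter so stronger hits play louder and lower in pitch.
# A root modulator adds slight random pitch and volume variation.
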SoundCue /MotherGameplay/Sounds/Boulder/Boulder_Impact_Phys_Cue
root=Modulator_1
nodes [
 {
  id=ModulatorContinuous_0
  children [Random_1]
  properties {
   pitchModulationParams {default=0, maxOutput=0.8, minOutput=1.2, parameterName=Strength}
   volumeModulationParams {default=0, minOutput=0.5, parameterName=Strength}
  }
 }
 {
  id=ModulatorContinuous_1
  children [Random_2]
  properties {
   pitchModulationParams {default=0, maxOutput=0.8, minOutput=1.2, parameterName=Strength}
   volumeModulationParams {default=0, maxOutput=1.2, minOutput=0.8, parameterName=Strength}
  }
 }
 {
  id=ModulatorContinuous_2
  children [Random_5]
  properties {
   pitchModulationParams {default=0, maxOutput=0.8, minOutput=1.2, parameterName=Strength}
   volumeModulationParams {default=0, maxOutput=1.2, minOutput=0.8, parameterName=Strength}
  }
 }
 {id=Modulator_1, children [Switch_0], properties {pitchMax=1.1, pitchMin=0.9, volumeMax=1.1, volumeMin=0.9}}
 {id=Random_1, children [WavePlayer_3,WavePlayer_4,WavePlayer_5,WavePlayer_6,WavePlayer_7,WavePlayer_8], properties {weights [1,1,1,1,1,1]}}
 {
  id=Random_2
  children [WavePlayer_9, WavePlayer_10, WavePlayer_11, WavePlayer_12, WavePlayer_13, WavePlayer_14, WavePlayer_15]
  properties {weights [1,1,1,1,1,1,1]}
 }
 {
  id=Random_5
  children [WavePlayer_16, WavePlayer_17, WavePlayer_18, WavePlayer_19, WavePlayer_20, WavePlayer_21, WavePlayer_22, WavePlayer_23]
  properties {weights [1,1,1,1,1,1,1,1]}
 }
 {
  id=Switch_0
  children [ModulatorContinuous_0, ModulatorContinuous_0, ModulatorContinuous_1, ModulatorContinuous_2]
  properties {intParameterName=ImpactType}
 }
 {id=WavePlayer_10, properties {soundWaveAssetPtr=&/Boulder_Med_Bounce_02}}
 {id=WavePlayer_11, properties {soundWaveAssetPtr=&/Boulder_Med_Bounce_03}}
 {id=WavePlayer_12, properties {soundWaveAssetPtr=&/Boulder_Med_Bounce_04}}
 {id=WavePlayer_13, properties {soundWaveAssetPtr=&/Boulder_Med_Bounce_05}}
 {id=WavePlayer_14, properties {soundWaveAssetPtr=&/Boulder_Med_Bounce_06}}
 {id=WavePlayer_15, properties {soundWaveAssetPtr=&/Boulder_Med_Bounce_07}}
 {id=WavePlayer_16, properties {soundWaveAssetPtr=&/Boulder_Large_Bounce_01}}
 {id=WavePlayer_17, properties {soundWaveAssetPtr=&/Boulder_Large_Bounce_02}}
 {id=WavePlayer_18, properties {soundWaveAssetPtr=&/Boulder_Large_Bounce_03}}
 {id=WavePlayer_19, properties {soundWaveAssetPtr=&/Boulder_Large_Bounce_04}}
 {id=WavePlayer_20, properties {soundWaveAssetPtr=&/Boulder_Large_Bounce_05}}
 {id=WavePlayer_21, properties {soundWaveAssetPtr=&/Boulder_Large_Bounce_06}}
 {id=WavePlayer_22, properties {soundWaveAssetPtr=&/Boulder_Large_Bounce_07}}
 {id=WavePlayer_23, properties {soundWaveAssetPtr=&/Boulder_Large_Bounce_08}}
 {id=WavePlayer_3, properties {soundWaveAssetPtr=&/Boulder_Small_Bounce_01}}
 {id=WavePlayer_4, properties {soundWaveAssetPtr=&/Boulder_Small_Bounce_02}}
 {id=WavePlayer_5, properties {soundWaveAssetPtr=&/Boulder_Small_Bounce_03}}
 {id=WavePlayer_6, properties {soundWaveAssetPtr=&/Boulder_Small_Bounce_04}}
 {id=WavePlayer_7, properties {soundWaveAssetPtr=&/Boulder_Small_Bounce_05}}
 {id=WavePlayer_8, properties {soundWaveAssetPtr=&/Boulder_Small_Bounce_06}}
 {id=WavePlayer_9, properties {soundWaveAssetPtr=&/Boulder_Med_Bounce_01}}
]
properties {
 attenuationSettings=&/Boulder_Physics_Impact_Att
 maxDistance=10500
 soundClassObject=&Fort_Impacts
 volumeMultiplier=1.25
}
