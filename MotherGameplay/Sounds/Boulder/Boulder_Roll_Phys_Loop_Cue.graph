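# Sound cue for the looping roll sound of a physics boulder.
# A looping wave player plays the slow boulder loop into a continuous modulator at the root.
# The modulator reads a runtime Speed parameter to vary both volume and pitch, with pitch kept in a narrow band around normal.
# It uses the boulder physics impact attenuation and the impacts sound class.
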SoundCue /MotherGameplay/Sounds/Boulder/Boulder_Roll_Phys_Loop_Cue
root=ModulatorContinuous_0
nodes [
 {
  id=ModulatorContinuous_0
  children [WavePlayer_1]
  properties {
   pitchModulationParams {default=0, maxOutput=1.2, minOutput=0.8, parameterName=Speed}
   volumeModulationParams {default=0, parameterName=Speed}
  }
 }
 {id=WavePlayer_1, properties {bLooping=true, soundWaveAssetPtr=&/Boulder_Loop_Slow_Loop}}
]
properties {
 attenuationSettings=&/Boulder_Physics_Impact_Att
 maxDistance=10500
 soundClassObject=&Fort_Impacts
 volumeMultiplier=1
}
